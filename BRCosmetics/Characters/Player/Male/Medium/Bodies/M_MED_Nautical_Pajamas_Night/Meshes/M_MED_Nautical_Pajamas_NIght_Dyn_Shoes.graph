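# Physics asset for the shoes of the male medium Nautical Pajamas Night cosmetic body, covering dynamic tongue and eye bones on both feet.
# Each ball bone anchors a two-box tongue chain and two eye pairs, each a small master capsule constrained to a larger eye capsule.
# Master-to-ball joints rotate freely, while eye-to-master joints have tight angular limits.
# The tongues are asymmetric: left joints are angularly free, while right joints use tighter, non-free limits.
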PhysicsAsset /BRCosmetics/Characters/Player/Male/Medium/Bodies/M_MED_Nautical_Pajamas_Night/Meshes/M_MED_Nautical_Pajamas_NIght_Dyn_Shoes
bodies {
 ball_l [{shapeType=Capsule, length=1, radius=1}]
 ball_r [{shapeType=Capsule, length=1, radius=1}]
 dyn_eye_01_l [{shapeType=Capsule, length=2, radius=2.5, center=<2 0 0>, rotation=<rot 90 0 0>}]
 dyn_eye_01_l_MSTR [{shapeType=Capsule, length=0.5, radius=0.5}]
 dyn_eye_01_r [{shapeType=Capsule, length=2, radius=2.5, center=<2 0 0>, rotation=<rot 90 0 0>}]
 dyn_eye_01_r_MSTR [{shapeType=Capsule, length=0.5, radius=0.5}]
 dyn_eye_02_l [{shapeType=Capsule, length=2, radius=2.5, center=<2 0 0>, rotation=<rot 90 0 0>}]
 dyn_eye_02_l_MSTR [{shapeType=Capsule, length=0.5, radius=0.5}]
 dyn_eye_02_r [{shapeType=Capsule, length=2, radius=2.5, center=<2 0 0>, rotation=<rot 90 0 0>}]
 dyn_eye_02_r_MSTR [{shapeType=Capsule, length=0.5, radius=0.5}]
 dyn_shoes_tongue_01_l [{shapeType=Box, extentX=3, extentY=6, extentZ=2, center=<1 0 0>}]
 dyn_shoes_tongue_01_r [{shapeType=Box, extentX=3, extentY=6, extentZ=2, center=<1 0 0>}]
 dyn_shoes_tongue_02_l [{shapeType=Box, extentX=3, extentY=6, extentZ=2, center=<1 0 0>}]
 dyn_shoes_tongue_02_r [{shapeType=Box, extentX=3, extentY=6, extentZ=2, center=<1 0 0>}]
}
constraints [
 {
  bone1Name=dyn_shoes_tongue_01_l
  bone2Name=ball_l
  swing1LimitDegrees=45
  swing1Motion=Free
  swing2LimitDegrees=45
  swing2Motion=Free
  twistLimitDegrees=45
  twistMotion=Free
 }
 {
  bone1Name=dyn_shoes_tongue_02_l
  bone2Name=dyn_shoes_tongue_01_l
  swing1LimitDegrees=45
  swing1Motion=Free
  swing2LimitDegrees=45
  swing2Motion=Free
  twistLimitDegrees=45
  twistMotion=Free
 }
 {
  bone1Name=dyn_eye_01_l_MSTR
  bone2Name=ball_l
  swing1LimitDegrees=45
  swing1Motion=Free
  swing2LimitDegrees=45
  swing2Motion=Free
  twistLimitDegrees=45
  twistMotion=Free
 }
 {bone1Name=dyn_eye_01_l, bone2Name=dyn_eye_01_l_MSTR, swing1LimitDegrees=20, swing2LimitDegrees=20, twistLimitDegrees=5}
 {
  bone1Name=dyn_eye_02_l_MSTR
  bone2Name=ball_l
  swing1LimitDegrees=45
  swing1Motion=Free
  swing2LimitDegrees=45
  swing2Motion=Free
  twistLimitDegrees=45
  twistMotion=Free
 }
 {bone1Name=dyn_eye_02_l, bone2Name=dyn_eye_02_l_MSTR, swing1LimitDegrees=20, swing2LimitDegrees=20, twistLimitDegrees=5}
 {bone1Name=dyn_shoes_tongue_01_r, bone2Name=ball_r, swing1LimitDegrees=15, swing2LimitDegrees=30, twistLimitDegrees=2}
 {bone1Name=dyn_shoes_tongue_02_r, bone2Name=dyn_shoes_tongue_01_r, swing1LimitDegrees=15, swing2LimitDegrees=30, twistLimitDegrees=2}
 {
  bone1Name=dyn_eye_01_r_MSTR
  bone2Name=ball_r
  swing1LimitDegrees=45
  swing1Motion=Free
  swing2LimitDegrees=45
  swing2Motion=Free
  twistLimitDegrees=45
  twistMotion=Free
 }
 {bone1Name=dyn_eye_01_r, bone2Name=dyn_eye_01_r_MSTR, swing1LimitDegrees=20, swing2LimitDegrees=20, twistLimitDegrees=5}
 {
  bone1Name=dyn_eye_02_r_MSTR
  bone2Name=ball_r
  swing1LimitDegrees=45
  swing1Motion=Free
  swing2LimitDegrees=45
  swing2Motion=Free
  twistLimitDegrees=45
  twistMotion=Free
 }
 {bone1Name=dyn_eye_02_r, bone2Name=dyn_eye_02_r_MSTR, swing1LimitDegrees=20, swing2LimitDegrees=20, twistLimitDegrees=5}
]
