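# Post-copy animation blueprint for the Night variant of the male medium Nautical Pajamas body cosmetic, adding secondary motion to the zipper and shoes.
# It copies the attached parent's pose, then constrains the zipper master bone to upper-body bones and four shoe "eye" master bones to foot and ball bones.
# Three RigidBody nodes then simulate each shoe and the zipper in base-bone space with custom gravity, reusing the base Pajamas zipper physics asset.
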AnimBlueprint /BRCosmetics/Characters/Player/Male/Medium/Bodies/M_MED_Nautical_Pajamas_Night/Meshes/M_MED_Nautical_Pajamas_Night_AnimBP
target_skeleton=&/M_MED_Nautical_Pajamas_Night_Skeleton
graphs [
 AnimGraph {nodes [
   n1 = "Constraint - dyn_zipper_MSTR (PPPPP)"!(
    ComponentPose=Local To Component(LocalPose=Copy Pose From Mesh() {bUseAttachedParent=true})
    Alpha=1
    ConstraintWeights_0=0.0242
    ConstraintWeights_1:float=0
    ConstraintWeights_2:float=0.001
    ConstraintWeights_3:float=0.9465
    ConstraintWeights_4:float=0.0283
   ) -> (Pose) {
    alphaBoolBlend {blendOption=Linear}
    boneToModify {boneName=dyn_zipper_MSTR}
    constraintSetup [
     {targetBone {boneName=clavicle_r}}
     {targetBone {boneName=neck_01}}
     {targetBone {boneName=spine_04}}
     {targetBone {boneName=spine_05}}
     {targetBone {boneName=clavicle_l}}
    ]
    constraintWeights [1,1,1,1,1]
   }
   n2 = "Constraint - dyn_eye_01_r_MSTR (PP)"!(ComponentPose=@n1, Alpha=1, ConstraintWeights_0=0.6875, ConstraintWeights_1:float=0.3125) -> (Pose) {
    alphaBoolBlend {blendOption=Linear}
    boneToModify {boneName=dyn_eye_01_r_MSTR}
    constraintSetup [{targetBone {boneName=ball_r}}, {targetBone {boneName=foot_r}}]
    constraintWeights [1,1]
   }
   n3 = "Constraint - dyn_eye_02_r_MSTR (PP)"!(ComponentPose=@n2, Alpha=1, ConstraintWeights_0=0.847, ConstraintWeights_1:float=0.153) -> (Pose) {
    alphaBoolBlend {blendOption=Linear}
    boneToModify {boneName=dyn_eye_02_r_MSTR}
    constraintSetup [{targetBone {boneName=ball_r}}, {targetBone {boneName=foot_r}}]
    constraintWeights [1,1]
   }
   n4 = "Constraint - dyn_eye_01_l_MSTR (PP)"!(ComponentPose=@n3, Alpha=1, ConstraintWeights_0=0.6875, ConstraintWeights_1:float=0.3125) -> (Pose) {
    alphaBoolBlend {blendOption=Linear}
    boneToModify {boneName=dyn_eye_01_l_MSTR}
    constraintSetup [{targetBone {boneName=ball_l}}, {targetBone {boneName=foot_l}}]
    constraintWeights [1,1]
   }
   n5 = "Constraint - dyn_eye_02_l_MSTR (PP)"!(ComponentPose=@n4, Alpha=1, ConstraintWeights_0=0.847, ConstraintWeights_1:float=0.153) -> (Pose) {
    alphaBoolBlend {blendOption=Linear}
    boneToModify {boneName=dyn_eye_02_l_MSTR}
    constraintSetup [{targetBone {boneName=ball_l}}, {targetBone {boneName=foot_l}}]
    constraintWeights [1,1]
   }
   n6 = RigidBody(ComponentPose=@n5) {
    bClampLinearTranslationLimitToRefPose=true
    bOverrideWorldGravity=true
    overridePhysicsAsset=&/M_MED_Nautical_Pajamas_NIght_Dyn_Shoes_Left
    simulationSpace=BaseBoneSpace
    alphaBoolBlend {blendOption=Linear}
    baseBoneRef {boneName=foot_l}
    componentAppliedLinearAccClamp=<3000 3000 3000>
    componentLinearAccScale=<0.4 0.4 0.4>
    componentLinearVelScale=<0.2 0.2 0.2>
    overrideWorldGravity {z=-580}
   }
   n7 = RigidBody(ComponentPose=@n6) {
    bClampLinearTranslationLimitToRefPose=true
    bOverrideWorldGravity=true
    overridePhysicsAsset=&/M_MED_Nautical_Pajamas_NIght_Dyn_Shoes
    simulationSpace=BaseBoneSpace
    alphaBoolBlend {blendOption=Linear}
    baseBoneRef {boneName=foot_r}
    componentAppliedLinearAccClamp=<3000 3000 3000>
    componentLinearAccScale=<0.4 0.4 0.4>
    componentLinearVelScale=<0.2 0.2 0.2>
    overrideWorldGravity {z=-580}
   }
   n8 = RigidBody(ComponentPose=@n7) {
    bClampLinearTranslationLimitToRefPose=true
    bOverrideWorldGravity=true
    overridePhysicsAsset=/BRCosmetics/Characters/Player/Male/Medium/Bodies/M_MED_Nautical_Pajamas/Meshes/M_MED_Nautical_Pajamas_Dyn_Zipper
    simulationSpace=BaseBoneSpace
    alphaBoolBlend {blendOption=Linear}
    baseBoneRef {boneName=spine_05}
    componentAppliedLinearAccClamp=<2000 2000 2000>
    componentLinearAccScale=<0.35 0.35 0.35>
    componentLinearVelScale=<0.3 0.3 0.3>
    overrideWorldGravity {z=-980}
   }
   Output Pose(Result=Component To Local(ComponentPose=@n8)) {layerGroup=None}
  ]}
 EventGraph {nodes [Event BlueprintUpdateAnimation()]}
]
variables []
